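# Niagara environment effect for short oil drips: small, short-lived sprites that fall straight down.
# The GPU Drip_And_Splash emitter spawns intermittently via probabilistic spawn rate, using fixed downward world velocity with gravity disabled.
# Collision with no bounce and full friction stops drops on contact, while speed-based scaling stretches them in flight.
# A Decal emitter for a flat, dark ground sprite is present but disabled.
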NiagaraSystem /Game/Effects/Fort_Effects/Effects/Environment/Apollo/NS_OilDrip_Short
dependencies {usedDataInterfaces [&NiagaraDataInterfaceCurve], usedDynamicInputs [], usedModules [&SystemState]}
emitters {
 Decal {
  bEnabled=false
  emitterUpdateScript [
   &EmitterState(Life Cycle Mode=::NewEnumerator1, Loop Duration=1, UseLoopDelay=false)
   &SpawnBurst_Instantaneous(Spawn Count=1, Use Spawn Probability=false)
  ]
  particleSpawnScript [
   /Niagara/Modules/Spawn/Initialization/InitializeParticle(
    Color=<rgba 0.0041 0.0041 0.0052 1>
    Lifetime=3
    Mass=&UniformRangedFloat(Minimum=0.75, Maximum=1.75)
    Module.Write Color=true
    Module.Write Lifetime=false
    Module.Write Mass=false
    Module.Write Material Random=true
    Module.Write Position=false
    Module.Write Scale=false
    Module.Write SpriteRotation=false
    Module.Write SpriteSize=true
    Position=&SimulationPosition()
    Sprite Size=<45 45>
   )
   &AddVelocity(Velocity=<0 0 -200>) {enabled=false}
  ]
  particleUpdateScript [
   &ParticleState(Kill Particles When Lifetime Has Elapsed=false, Let Infinitely Lived Particles Die When Emitter Deactivates=true)
   /NiagaraFN/Sprite/FN_SpriteAlignmentAndFacing(SpriteAlignment:Vector3f=<0 1 0>, SpriteFacing:Vector3f=<0 0 1>) {module=SpriteAlignmentAndFacing}
  ]
  rendererClasses [&NiagaraSpriteRendererProperties]
  renderers [{rendererClass=&NiagaraSpriteRendererProperties}]
 }
 Drip_And_Splash {
  simTarget=GPUComputeSim
  emitterUpdateScript [
   &EmitterState(
    Life Cycle Mode=::NewEnumerator1
    Loop Delay=&UniformRangedFloat(Minimum=0, Maximum=0.5)
    Loop Duration=1
    UseLoopDelay=false
   )
   &SpawnBurst_Instantaneous(Spawn Count=1, Spawn Probability=0.5, Use Spawn Probability=true) {enabled=false}
   &SpawnRate(Spawn Probability=0.5, SpawnRate=3, Use Spawn Probability=true)
  ]
  particleSpawnScript [
   /SetVariables(Particles.Random=&UniformRangedFloat(Minimum=0, Maximum=1)) {enabled=false, isSetParameters=true}
   &InitializeParticle(
    Lifetime=0.25
    Mass=&UniformRangedFloat(Minimum=1, Maximum=1)
    Module.Write Color=true
    Module.Write Lifetime=true
    Module.Write Mass=false
    Module.Write Material Random=true
    Module.Write Position=false
    Module.Write Scale=false
    Module.Write SpriteRotation=false
    Module.Write SpriteSize=true
    Position=&SimulationPosition()
    Sprite Size=<2 2>
   )
   /SetVariables(Particles.UVScale:Vector2f=&MakeVector2D(
     X=&FloatFromCurve(FloatCurve=DataInterface() {propertyValues {Curve {keys [<at 0 1 RCIM_Constant>,<at 0.5 -1>]}}}, CurveIndex=$Particles.Random)
     Y=&FloatFromCurve(FloatCurve=DataInterface() {propertyValues {Curve {keys [<at 0 1 RCIM_Constant>,<at 0.5 -1>]}}}, CurveIndex=$Particles.Random)
    )) {enabled=false, isSetParameters=true}
   &AddVelocity(Coordinate Space=::World, Velocity=<0 0 -300>)
  ]
  particleUpdateScript [
   &ParticleState()
   &GravityForce() {enabled=false}
   &Collision(
    Advanced Aging Rate=1
    Age Colliding Particles=false
    Friction=1
    Friction During a Bounce=1
    Kill Occluded Particles=false
    Kill Offscreen Particles=true
    Maximum time offscreen=0.2
    Module.Control Rest State Via Surface Normal=false
    Module.EnableMaxCollisionCount=false
    Module.Randomize Collision Normal Vector=false
    Particle Radius Scale=0
    Restitution=0
   )
   &SolveForcesAndVelocity(Clamp Velocity=false, Limit Acceleration=false)
   &ScaleSpriteSizeBySpeed(Max Scale Factor=<1 3>, Velocity Threshold=200)
   &ScaleColor(
    Scale Alpha=&FloatFromCurve(FloatCurve=DataInterface() {propertyValues {Curve {keys [<at 0 0>,<at 0.1 1>,<at 0.9 1>,<at 1.0011 0.002>]}}})
    ScaleA=true
    ScaleRGB=true
   ) {enabled=false}
  ]
  rendererClasses [&NiagaraSpriteRendererProperties]
  renderers [{rendererClass=&NiagaraSpriteRendererProperties}]
 }
}
summary {
 systemName=NS_OilDrip_Short
 emitters [
  {bEnabled=false, emitterName=Decal, rendererClasses [&NiagaraSpriteRendererProperties]}
  {emitterName=Drip_And_Splash, simTarget=GPUComputeSim, rendererClasses [&NiagaraSpriteRendererProperties]}
 ]
 userVariables [User.SpriteRotation {type=NiagaraFloat, defaultValue=180, description=""}]
}
